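# Physics asset for the Masked Warrior Spring backpack cosmetic, setting up its dangling charm for physical motion.
# A large capsule covers the pack, and small capsules cover the charm base and five chained charm bones.
# Constraints link the chain in series from the pack through the base to the tip, with twist locked and tight swing limits that loosen toward the end.
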PhysicsAsset /BRCosmetics/Accessories/FORT_Backpacks/Backpack_M_MED_Masked_Warrior_Spring/Meshes/M_MED_Masked_Warrior_Spring_Pack_Physics
bodies {
 dyn_Charm_01 [{
   shapeType=Capsule
   length=3.25
   radius=1
   shapeName=dyn_Charm_01_capsule
   center=<1.5899 0 -0.0406>
   rotation=<rot -88.5306 -179.9973 179.9973>
  }]
 dyn_Charm_02 [{
   shapeType=Capsule
   length=3.25
   radius=1
   shapeName=dyn_Charm_02_capsule
   center=<1.5899 0 -0.0402>
   rotation=<rot -88.5441 -179.9968 179.9968>
  }]
 dyn_Charm_03 [{
   shapeType=Capsule
   length=3.25
   radius=1
   shapeName=dyn_Charm_03_capsule
   center=<1.5899 0 -0.0396>
   rotation=<rot -88.5734 -179.9964 179.9964>
  }]
 dyn_Charm_04 [{
   shapeType=Capsule
   length=3.25
   radius=1
   shapeName=dyn_Charm_04_capsule
   center=<1.59 0 -0.0364>
   rotation=<rot -88.6891 -179.9996 179.9996>
  }]
 dyn_Charm_05 [{
   shapeType=Capsule
   length=3.25
   radius=1
   shapeName=dyn_Charm_05_capsule
   center=<1.59 0 -0.0364>
   rotation=<rot -88.6891 -179.9996 179.9996>
  }]
 dyn_Charm_base [{shapeType=Capsule, length=0.75, radius=1.25, shapeName=dyn_Charm_base_capsule, rotation=<rot 50.399 -90 0>}]
 pack [{shapeType=Capsule, length=38.7081, radius=19.9947, center=<1.9853 12.4079 0.8561>, rotation=<rot 1.9789 -89.9693 -0.6182>}]
}
constraints [
 {bone1Name=dyn_Charm_base, bone2Name=pack, swing1LimitDegrees=5, swing2LimitDegrees=5, twistLimitDegrees=0, twistMotion=Locked}
 {bone1Name=dyn_Charm_01, bone2Name=dyn_Charm_base, swing1LimitDegrees=5, swing2LimitDegrees=10, twistLimitDegrees=0, twistMotion=Locked}
 {bone1Name=dyn_Charm_02, bone2Name=dyn_Charm_01, swing1LimitDegrees=5, swing2LimitDegrees=10, twistLimitDegrees=0, twistMotion=Locked}
 {bone1Name=dyn_Charm_03, bone2Name=dyn_Charm_02, swing1LimitDegrees=5, swing2LimitDegrees=10, twistLimitDegrees=0, twistMotion=Locked}
 {bone1Name=dyn_Charm_04, bone2Name=dyn_Charm_03, swing1LimitDegrees=8, swing2LimitDegrees=15, twistLimitDegrees=0, twistMotion=Locked}
 {bone1Name=dyn_Charm_05, bone2Name=dyn_Charm_04, swing1LimitDegrees=10, swing2LimitDegrees=20, twistLimitDegrees=0, twistMotion=Locked}
]
